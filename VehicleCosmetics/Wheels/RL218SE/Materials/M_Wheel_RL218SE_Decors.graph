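# Translucent lit material for decor details on the RL218SE vehicle cosmetic wheel, enabled for skeletal meshes and mesh deformers.
# A base color texture is tinted by a Color parameter, alongside a normal map that defaults to flat.
# The mask's green channel scales the emissive color and amount; the Fresnel input has no effect because the lerp alpha is fixed at 1.
# A static switch selects either red-channel masked opacity or uniform scalar opacity.
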Material /VehicleCosmetics/Wheels/RL218SE/Materials/M_Wheel_RL218SE_Decors
graph {
 properties {
  bAutomaticallySetUsageInEditor=false
  bUsedWithMeshDeformer=true
  bUsedWithSkeletalMesh=true
  blendMode=BLEND_Translucent
  translucencyLightingMode=TLM_VolumetricDirectional
  usedShadingModels=Default Lit
 }
 nodes [
  Multiply(
   A=TextureSampleParameter2D() {parameterName=Basecolor}.RGB
   B=VectorParameter() {group=Custom Color Properties, parameterName=Color, defaultValue=<rgba 0.59 0 0 1>}.RGB
  )
  n1 = TextureSampleParameter2D() {parameterName=Mask, samplerType=SAMPLERTYPE_Masks, texture=/VehicleCosmetics/Wheels/RL218SE/Textures/T_Decors_218_M}
  n2 = VectorParameter() {group=Custom Color Properties, parameterName=EmissiveColor, defaultValue=<rgba 0.59 0 0 1>}
  n3 = ScalarParameter() {defaultValue=1, group=Base Parameters, parameterName=Emissive Amount, sortPriority=5}
  Multiply(A=@n1.G, B=LinearInterpolate(A=Fresnel(), B=Multiply(A=@n2.RGB, B=@n3)) {constAlpha=1})
  n4 = ScalarParameter() {defaultValue=1.5, group=Base Parameters, parameterName=Opacity, sortPriority=3}
  StaticSwitchParameter(True=Multiply(A=@n1.R, B=@n4), False=@n4) {group=Base Parameters, parameterName=Use Mask Opacity, sortPriority=1}
  TextureSampleParameter2D() {
   group=Base Textures
   parameterName=Normal
   samplerType=SAMPLERTYPE_Normal
   sortPriority=1
   texture=&FlatNormalBlue
  }
 ]
}
